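; Test actor that defines a cone-shaped volume for a storm screen-effect mesh, configured by height, top/bottom radii and blend distances.
; The construction script feeds that shape into a dynamic material, using an override material when valid, and enlarges mesh bounds to fit it.
; With Debug on, two cylinder materials show the core and blended shapes; otherwise the main debug cylinder is destroyed.
; BeginPlay, Tick and overlap events are empty.
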(event EventBeginPlay)

(event Collision|EventActorBeginOverlap
  (params (OtherActor :type Actor%20Object%20Reference)))

(event EventTick
  (params (DeltaSeconds :type Float%20%28single-precision%29)))

(fn ConstructionScript
  (bind (_min _max) (Components|StaticMesh|GetLocalBounds (Variables|Default|GetSM_Storm_ScreenEffect)))
  (bind (_min_2 _max_2) (Components|StaticMesh|GetLocalBounds (Variables|Default|GetDebug_Cylinder)))
  (Utilities|FlowControl|Sequence
    (:then_0
      (if (Variables|Default|GetDebug)
        (Utilities|FlowControl|Sequence
          (:then_0
            (bind _returnvalue (Rendering|Material|CreateDynamicMaterialInstance (Variables|Default|GetDebug_Cylinder) 0 "/Figment_S10/Athena/Events/Test/M_DebugConeShape.M_DebugConeShape" "None" :!class /Script/Engine.PrimitiveComponent))
            (Rendering|Material|SetScalarParameterValue _returnvalue "HeightTop" (Variables|Default|GetHeight) :!class /Script/Engine.MaterialInstanceDynamic)
            (Rendering|Material|SetScalarParameterValue _returnvalue "HeightBottom" (Variables|Default|GetHeight) :!class /Script/Engine.MaterialInstanceDynamic)
            (Rendering|Material|SetScalarParameterValue _returnvalue "RadiusTop" (Variables|Default|GetRadiusTop) :!class /Script/Engine.MaterialInstanceDynamic)
            (Rendering|Material|SetScalarParameterValue _returnvalue "RadiusBottom" (Variables|Default|GetRadiusBottom) :!class /Script/Engine.MaterialInstanceDynamic)
            (Rendering|SetBoundsScale (Variables|Default|GetDebug_Cylinder) (/ (Math|Float|Max(Float) (Math|Float|float+float (Variables|Default|GetHeight) (Variables|Default|GetBlendDistance_Top) (+ (Variables|Default|GetRadiusTop) (Variables|Default|GetBlendDistance)) (Variables|Default|GetBlendDistance)) (Math|Float|float+float (Variables|Default|GetHeight) (Variables|Default|GetBlendDistance_Bottom) (+ (Variables|Default|GetRadiusBottom) (Variables|Default|GetBlendDistance)) (Variables|Default|GetBlendDistance))) (Math|Float|Max(Float) (.x _max_2) (.y _max_2) (.z _max_2)))))
          (:then_1
            (bind _returnvalue_2 (Rendering|Material|CreateDynamicMaterialInstance (Variables|Default|GetDebug_Cylindertranslucent) 0 "/Figment_S10/Athena/Events/Test/M_DebugConeShape_Trans.M_DebugConeShape_Trans" "None" :!class /Script/Engine.PrimitiveComponent))
            (Rendering|Material|SetScalarParameterValue _returnvalue_2 "HeightTop" (+ (Variables|Default|GetHeight) (Variables|Default|GetBlendDistance_Top)) :!class /Script/Engine.MaterialInstanceDynamic)
            (Rendering|Material|SetScalarParameterValue _returnvalue_2 "HeightBottom" (+ (Variables|Default|GetHeight) (Variables|Default|GetBlendDistance_Bottom)) :!class /Script/Engine.MaterialInstanceDynamic)
            (Rendering|Material|SetScalarParameterValue _returnvalue_2 "RadiusTop" (+ (Variables|Default|GetRadiusTop) (Variables|Default|GetBlendDistance)) :!class /Script/Engine.MaterialInstanceDynamic)
            (Rendering|Material|SetScalarParameterValue _returnvalue_2 "RadiusBottom" (+ (Variables|Default|GetRadiusBottom) (Variables|Default|GetBlendDistance)) :!class /Script/Engine.MaterialInstanceDynamic)))
        (else
          (Components|DestroyComponent (Variables|Default|GetDebug_Cylinder)))))
    (:then_1
      (bind _returnvalue_3 (Rendering|Material|CreateDynamicMaterialInstance (Variables|Default|GetSM_Storm_ScreenEffect) 0 (select (Utilities|IsValid (Variables|Default|GetOverideMaterial) :!class /Script/Engine.KismetSystemLibrary) (Variables|Default|GetOverideMaterial) (Rendering|Material|GetMaterial (Variables|Default|GetSM_Storm_ScreenEffect) 0)) "None" :!class /Script/Engine.PrimitiveComponent))
      (Rendering|Material|SetScalarParameterValue _returnvalue_3 "F_RadiusTop" (Variables|Default|GetRadiusTop) :!class /Script/Engine.MaterialInstanceDynamic)
      (Rendering|Material|SetScalarParameterValue _returnvalue_3 "F_BlendRadius" (Variables|Default|GetBlendDistance) :!class /Script/Engine.MaterialInstanceDynamic)
      (Rendering|Material|SetScalarParameterValue _returnvalue_3 "F_RadiusBottom" (Variables|Default|GetRadiusBottom) :!class /Script/Engine.MaterialInstanceDynamic)
      (Rendering|Material|SetScalarParameterValue _returnvalue_3 "F_Height" (Variables|Default|GetHeight) :!class /Script/Engine.MaterialInstanceDynamic)
      (Rendering|Material|SetScalarParameterValue _returnvalue_3 "F_BlendTop" (Variables|Default|GetBlendDistance_Top) :!class /Script/Engine.MaterialInstanceDynamic)
      (Rendering|Material|SetScalarParameterValue _returnvalue_3 "F_BlendBottom" (Variables|Default|GetBlendDistance_Bottom) :!class /Script/Engine.MaterialInstanceDynamic)
      (Rendering|SetBoundsScale (Variables|Default|GetSM_Storm_ScreenEffect) (/ (Math|Float|Max(Float) (Math|Float|float+float (Variables|Default|GetHeight) (Variables|Default|GetBlendDistance_Top) (+ (Variables|Default|GetRadiusTop) (Variables|Default|GetBlendDistance)) (Variables|Default|GetBlendDistance)) (Math|Float|float+float (Variables|Default|GetHeight) (Variables|Default|GetBlendDistance_Bottom) (+ (Variables|Default|GetRadiusBottom) (Variables|Default|GetBlendDistance)) (Variables|Default|GetBlendDistance))) (Math|Float|Max(Float) (.x _max) (.y _max) (.z _max)))))))
