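# Translucent debug material that, judging by its name and math, reshapes a mesh into an adjustable cone for test visualization.
# Vertex color red blends between bottom and top radius and height parameters, building a vertex offset radially around and along the object's local Z axis.
# Camera distance along that axis drives a Color-to-blue blend, and its distance from the axis is also computed.
# No material output connections appear in the graph.
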Material /Figment_S10/Athena/Events/Test/M_DebugConeShape
graph {
 properties {
  blendMode=BLEND_Translucent
  pixelDepthOffsetMode=PDOM_Legacy
  translucencyLightingMode=TLM_VolumetricPerVertexDirectional
  usedShadingModels=Default Lit
 }
 nodes [
  n1 = VertexColor()
  n2 = LinearInterpolate(
   A=Multiply(A=ScalarParameter() {defaultValue=365.2841, parameterName=HeightBottom}) {constB=-1}
   B=ScalarParameter() {defaultValue=365.2841, parameterName=HeightTop}
   Alpha=@n1.R
  )
  AppendVector(A=Constant2Vector(), B=@n2)
  Constant() {r=0.5}
  n3 = Normalize(VectorInput=Transform(None=Constant3Vector() {constant=<rgba 0 0 1 0>}) {transformSourceType=TRANSFORMSOURCE_Local})
  n4 = ObjectPositionWS()
  n5 = Subtract(A=DotProduct(A=CameraPositionWS(), B=@n3), B=DotProduct(A=@n4, B=@n3))
  n6 = Multiply(A=@n5, B=@n3)
  Divide(A=Distance(A=Add(A=@n6, B=@n4), B=CameraPositionWS()), B=ScalarParameter() {defaultValue=2000, parameterName=Test2})
  LinearInterpolate(
   A=VectorParameter() {parameterName=Color, defaultValue=<rgba 1 0 0 1>}.RGB
   B=Constant3Vector() {constant=<rgba 0 0.1634 1 0>}
   Alpha=Divide(A=Abs(None=@n5), B=ScalarParameter() {defaultValue=2000, parameterName=Test})
  )
  ScalarParameter() {defaultValue=0.99, parameterName=Opacity}
  n7 = Transform(None=Subtract(A=ObjectPositionWS(), B=WorldPosition().XYZ)) {transformSourceType=TRANSFORMSOURCE_World, transformType=TRANSFORM_Local}
  n8 = Multiply(A=@n7, B=Constant3Vector() {constant=<rgba 1 1 0 0>})
  n9 = ScalarParameter() {defaultValue=200, parameterName=RadiusBottom}
  n10 = ScalarParameter() {defaultValue=500, parameterName=RadiusTop}
  n11 = Add(A=@n7, B=Multiply(A=Normalize(VectorInput=@n8), B=LinearInterpolate(A=@n9, B=@n10, Alpha=@n1.R)))
  n12 = Transform(None=Constant3Vector() {constant=<rgba 0 0 1 0>}) {transformSourceType=TRANSFORMSOURCE_Local}
  Transform(None=Add(A=Transform(None=@n11) {transformSourceType=TRANSFORMSOURCE_Local}, B=Multiply(A=@n12, B=@n2))) {transformSourceType=TRANSFORMSOURCE_Local}
 ]
}
